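; Looping gameplay cue notify that applies Victory Crown wear effects to the crown wearer and removes them when the cue ends.
; On loop start it passes particle components to ApplyCrownWear and, if a scalable-float toggle allows, starts a repeating timer at a configured ping frequency.
; Each tick shows a crown-icon sound indicator at the wearer's location, with a configured range and hostile, neutral and friendly affiliations.
; Removal calls RemoveCrownWear and clears that timer.
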(event EventOnLoopingStart(EitherVFX-preferthisversion)
  (params (MyTarget :type Actor%20Object%20Reference) (Parameters :type Gameplay%20Cue%20Parameters%20Structure%20%28by%20ref%29) (ParticleComponents :type Array%20of%20FXSystem%20Component%20Object%20References) (AudioComponents :type Array%20of%20Audio%20Component%20Object%20References))
  (CallFunction|ApplyCrownWear ParticleComponents)
  ;[1 1 1] Setup a looping AV Indicator on application
  (if (ScalableFloat|GetValueasBool (Variables|Parameters|GetAudioVizualizerShownOnWearer?))
    (bind _returnvalue (Utilities|Time|SetTimerbyEvent (delegate Custom|AVIndicatorLoop) (ScalableFloat|GetValueatLevel (Variables|Parameters|GetAudioVizualizerPingFrequency)) true))
    (Variables|Default|SetTimerHandle _returnvalue)))

(event Custom|AVIndicatorLoop
  ;[1 1 1] Setup a looping AV Indicator on application
  (Fortnite|ShowSoundIndicatorLocalforTeam :WorldContextObject (stale EventOnLoopingStart(EitherVFX-preferthisversion) MyTarget) :Location (Transformation|GetActorLocation (stale EventOnLoopingStart(EitherVFX-preferthisversion) MyTarget)) :MaxAudibleDistance (ScalableFloat|GetValueatLevel (Variables|Parameters|GetAudioVizualizerRange)) :IndicatorType "Chest" :Team (Team|GetActorTeam (stale EventOnLoopingStart(EitherVFX-preferthisversion) MyTarget)) :Affiliations (Utilities|Array|MakeArray "Hostile" "Neutral" "Friendly") :Instigator (stale EventOnLoopingStart(EitherVFX-preferthisversion) MyTarget) :Tint "(R=0.000000,G=0.000000,B=0.000000,A=0.000000)" :OverrideIcon "/Game/UI/Foundation/Shell/Textures/Shell-Top-Crown.Shell-Top-Crown" :AttenuationMaxRangeMultiplier 1.0 :AttenuationIntensityMultiplier 1.0))

(event EventOnRemoval(EitherVFX-preferthisversion)
  (params (MyTarget :type Actor%20Object%20Reference) (Parameters :type Gameplay%20Cue%20Parameters%20Structure%20%28by%20ref%29) (ParticleComponents :type Array%20of%20FXSystem%20Component%20Object%20References) (AudioComponents :type Array%20of%20Audio%20Component%20Object%20References) (BurstCameraShakeInstance :type Camera%20Shake%20Base%20Object%20Reference) (BurstDecalInstance :type Decal%20Actor%20Object%20Reference))
  (bind _timerhandle (Variables|Default|GetTimerHandle))
  (CallFunction|RemoveCrownWear)
  ;[1 1 1] Stop the AV Indicator loop on removal
  (if (Utilities|Time|IsValidTimerHandle _timerhandle)
    (Utilities|Time|ClearandInvalidateTimerbyHandle self _timerhandle)))

(fn ConstructionScript)
